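# Physics asset for the face accessory mesh of the male medium BriteDino cosmetic body, covering a helmet jaw and a tail.
# Capsule bodies on head and spine_05 sit alongside small capsules for each base and dynamic accessory bone.
# The jaw and four-link tail chain have narrow swing limits and zero-degree twist limits.
# Base bones attach to the head, and head to spine_05, with free angular motion.
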PhysicsAsset /BRCosmetics/Characters/Player/Male/Medium/Bodies/M_MED_BriteDino/Meshes/Parts/M_MED_BriteDino_FaceAcc_Physics
bodies {
 base_helmet_jaw [{shapeType=Capsule, length=1, radius=1, shapeName=base_helmet_jaw_capsule, rotation=<rot 90 0 0>}]
 base_tail [{shapeType=Capsule, length=1, radius=1, shapeName=base_tail_capsule, rotation=<rot 90 0 0>}]
 dyn_helmet_jaw [{shapeType=Capsule, length=1, radius=1, shapeName=dyn_helmet_jaw_capsule, rotation=<rot 90 0 0>}]
 dyn_tail_1 [{shapeType=Capsule, length=0.8, radius=2, shapeName=dyn_tail_1_capsule, center=<2 0 0>, rotation=<rot 90 0 0>}]
 dyn_tail_2 [{shapeType=Capsule, length=0.8, radius=2, shapeName=dyn_tail_2_capsule, center=<2 0 0>, rotation=<rot 90 0 0>}]
 dyn_tail_3 [{shapeType=Capsule, length=0.8, radius=2, shapeName=dyn_tail_3_capsule, center=<2 0 0>, rotation=<rot 90 0 0>}]
 dyn_tail_4 [{shapeType=Capsule, length=0.8, radius=2, shapeName=dyn_tail_3_capsule, center=<2 0 0>, rotation=<rot 90 0 0>}]
 head [{shapeType=Capsule, length=15, radius=6, shapeName=head_capsule, center=<0 0.7905 0>, rotation=<rot 90 0 0>}]
 spine_05 [{shapeType=Capsule, length=20, radius=9.5, shapeName=spine_05_capsule, center=<11 -1.8 0>, rotation=<rot 0 -90 0>}]
}
constraints [
 {bone1Name=dyn_helmet_jaw, bone2Name=base_helmet_jaw, swing1LimitDegrees=0, swing2LimitDegrees=10, twistLimitDegrees=0}
 {bone1Name=dyn_tail_1, bone2Name=base_tail, swing1LimitDegrees=5, swing2LimitDegrees=7.5, twistLimitDegrees=0}
 {bone1Name=dyn_tail_2, bone2Name=dyn_tail_1, swing1LimitDegrees=9, swing2LimitDegrees=7.5, twistLimitDegrees=0}
 {bone1Name=dyn_tail_3, bone2Name=dyn_tail_2, swing1LimitDegrees=9, swing2LimitDegrees=7.5, twistLimitDegrees=0}
 {
  bone1Name=base_helmet_jaw
  bone2Name=head
  swing1LimitDegrees=45
  swing1Motion=Free
  swing2LimitDegrees=45
  swing2Motion=Free
  twistLimitDegrees=45
  twistMotion=Free
 }
 {
  bone1Name=base_tail
  bone2Name=head
  swing1LimitDegrees=45
  swing1Motion=Free
  swing2LimitDegrees=45
  swing2Motion=Free
  twistLimitDegrees=45
  twistMotion=Free
 }
 {
  bone1Name=head
  bone2Name=spine_05
  swing1LimitDegrees=45
  swing1Motion=Free
  swing2LimitDegrees=45
  swing2Motion=Free
  twistLimitDegrees=45
  twistMotion=Free
 }
 {bone1Name=dyn_tail_4, bone2Name=dyn_tail_3, swing1LimitDegrees=9, swing2LimitDegrees=10, twistLimitDegrees=0}
]
