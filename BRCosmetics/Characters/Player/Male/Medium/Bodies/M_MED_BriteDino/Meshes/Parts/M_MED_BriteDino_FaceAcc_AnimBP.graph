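# Post-process style animation blueprint for the BriteDino character's face accessory, posing and simulating it, including its tail bone, atop the attached parent's pose.
# The copied pose gets a base_tail offset, then a Control Rig at half weight during emotes, then head-space RigidBody physics with overridden gravity.
# Additive base_tail transforms follow the simulation, driven by head tilt while crouching and by skydiving.
# The event graph updates these alphas and gravity, using a deprecated gravity function.
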AnimBlueprint /BRCosmetics/Characters/Player/Male/Medium/Bodies/M_MED_BriteDino/Meshes/Parts/M_MED_BriteDino_FaceAcc_AnimBP
target_skeleton=&/M_MED_BriteDino_FaceAcc_Skeleton
graphs [
 AnimGraph {nodes [
   "Transform (Modify) Bone - Bone: base_tail"(Alpha=1, Translation="0,0,0", Rotation="-30,0,0", Scale="1,1,1") {
    rotationMode=BMM_Additive
    rotationSpace=BCS_BoneSpace
    scaleMode=BMM_Additive
    scaleSpace=BCS_BoneSpace
    translationMode=BMM_Additive
    translationSpace=BCS_BoneSpace
    alphaBoolBlend {blendOption=Linear}
    boneToModify {boneName=base_tail}
    rotation {pitch=-30}
   }
   n1 = "Transform (Modify) Bone - Bone: base_tail"(
    ComponentPose=Local To Component(LocalPose=Copy Pose From Mesh() {bUseAttachedParent=true})
    Alpha=1
    Translation="1,0,0"
    Rotation="0,0,0"
    Scale="1,1,1"
   ) {
    rotationMode=BMM_Additive
    rotationSpace=BCS_BoneSpace
    scaleMode=BMM_Additive
    scaleSpace=BCS_BoneSpace
    translationMode=BMM_Additive
    translationSpace=BCS_BoneSpace
    alphaBoolBlend {blendOption=Linear}
    boneToModify {boneName=base_tail}
    translation {x=1}
   }
   n2 = Control Rig(Source=Component To Local(ComponentPose=@n1), Alpha=Get playingEmote() -> (playingEmote:double)) {
    alphaBoolBlend {blendOption=Linear}
    controlRigAssetReference {blueprintRigClass=&/M_MED_BriteDino_FaceAcc_CtrlRig.M_MED_BriteDino_FaceAcc_CtrlRig_C}
   }
   n3 = RigidBody!(ComponentPose=Local To Component(LocalPose=@n2), Alpha, OverrideWorldGravity:Vector=Get GravityOverride(), ExternalForce) -> (Pose) {
    bClampLinearTranslationLimitToRefPose=true
    bOverrideWorldGravity=true
    overridePhysicsAsset=&/M_MED_BriteDino_FaceAcc_Physics
    simulationSpace=BaseBoneSpace
    alphaBoolBlend {blendOption=Linear}
    baseBoneRef {boneName=head}
    componentAppliedLinearAccClamp=<3000 3000 3000>
    componentLinearAccScale=<0.3 0.3 0.2>
    componentLinearVelScale=<0.3 0.3 0.2>
    simSpaceSettings {
     maxAngularAcceleration=1000
     maxAngularVelocity=3
     maxLinearAcceleration=700
     maxLinearVelocity=500
     velocityScaleZ=0
     worldAlpha=0.225
     externalLinearDragV=<0.2 0.2 0.2>
    }
   }
   n4 = "Transform (Modify) Bone - Bone: base_tail"(ComponentPose=@n3, Alpha=Get headUp_CrouchAlpha() -> (headUp_CrouchAlpha:double), Translation="0,0,3", Rotation="0,25,0", Scale="1,1,1") {
    rotationMode=BMM_Additive
    rotationSpace=BCS_BoneSpace
    scaleMode=BMM_Additive
    scaleSpace=BCS_BoneSpace
    translationMode=BMM_Additive
    translationSpace=BCS_BoneSpace
    alphaBoolBlend {blendOption=Linear}
    boneToModify {boneName=base_tail}
    rotation {yaw=25}
    translation {z=3}
   }
   n5 = "Transform (Modify) Bone - Bone: base_tail"(ComponentPose=@n4, Alpha=Get skydiving_alpha() -> (skydiving_alpha:double), Translation="0,0,0", Rotation="28,0,0", Scale="1,1,1") {
    rotationMode=BMM_Additive
    rotationSpace=BCS_BoneSpace
    scaleMode=BMM_Additive
    scaleSpace=BCS_BoneSpace
    translationMode=BMM_Additive
    translationSpace=BCS_BoneSpace
    alphaBoolBlend {blendOption=Linear}
    boneToModify {boneName=base_tail}
    rotation {pitch=28}
   }
   Output Pose(Result=Component To Local(ComponentPose=@n5)) {layerGroup=None}
  ]}
 EventGraph {nodes [
   Event BlueprintUpdateAnimation()
   > n1 = Gravity Override State Function DEPRECATED(CharacterPartAnimInstance=Self-Reference(), GameTimeInSeconds=GetGameTimeInSeconds(), GravityOverrideParams=Get Gravity Override Params())
   > Set GravityOverride(GravityOverride=@n1)
   > Set SkelMesh(SkelMesh=GetPartSkeletalMeshComponent())
   > n2 = Set headUp(headUp:double=GetAlphaFromReferenceTransformDelta(SkeletalMeshComponent=Get SkelMesh(), BoneName=spine_05, BaseName=head, InputRange=(X=0,Y=120), TwistAxis="0,0,1")) -> (Output_Get:double)
   n3 = Get bIsCrouching()
   > Set headUp_CrouchAlpha(headUp_CrouchAlpha:double=float * float(A=@n2, B="To Float (Boolean)"(InBool=@n3)) {is=title K2Node_PromotableOperator}) -> (Output_Get:double)
   > Set skydiving_alpha(skydiving_alpha:double="To Float (Boolean)"(InBool=Get bIsSkyDiving())) -> (Output_Get:double)
   > n4 = Branch(Condition=Get bIsPlayingEmote())
   > Set playingEmote(playingEmote:double=0.5) -> (Output_Get:double)
   Set playingEmote(execute=@n4.else, playingEmote=1)
  ]}
]
variables [
 Gravity Override Params {type=GravityOverrideParamsStruct}
 GravityOverride {type=Vector}
 headUp {type=number}
 headUp_CrouchAlpha {type=number}
 playingEmote {type=number}
 SkelMesh {type=SkeletalMeshComponent}
 skydiving_alpha {type=number}
]
